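# UI material for the complete-setlist animation on the song results screen, drawing three translucent, looping square outlines that spin and grow.
# Each square feeds rotated UVs into MF_UI_Box, with MF_SparksUI_LinearTime driving angle and size at slightly offset speeds.
# The squares are built twice, as thin lines and with a glow, each set merged by Max.
# A Color parameter supplies the tint, and a soft circle fill masks a dimmed copy of the shapes.
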Material /PilgrimQuickplay/Widgets/Results/SongResults/CompleteSetlistAnimation/Materials_Results/M_UI_DiamondAnim
graph {
 properties {bAutomaticallySetUsageInEditor=false, blendMode=BLEND_Translucent, materialDomain=MD_UI, usedShadingModels=Default Lit}
 nodes [
  LinearInterpolate(A=NamedRerouteUsage(), Alpha=NamedRerouteUsage())
  n1 = TextureCoordinate()
  n2 = MaterialFunctionCall!(FloatValue, Speed=Constant() {r=0.11}) {materialFunction=&/MF_SparksUI_LinearTime}
  n3 = Constant() {r=0.33}
  NamedRerouteDeclaration(None=MaterialFunctionCall(
    UVs=MaterialFunctionCall(UVs=@n1, "Rotation Angle (0-1)"=@n2.0-1 Loop) {materialFunction=&CustomRotator}.Rotated Values
    Size=MaterialFunctionCall(Speed=@n3) {materialFunction=&MF_SparksUI_LinearTime}.0-1 Loop
    OutlineThickness=NamedRerouteUsage()
   ) {materialFunction=&MF_UI_Box}.Outline) {name=Square 1, nodeColor=<rgba 0.9922 0 1 1>}
  n4 = TextureCoordinate()
  n5 = MaterialFunctionCall(Speed=Constant() {r=0.12}) {materialFunction=&MF_SparksUI_LinearTime}
  n6 = Constant() {r=0.36}
  NamedRerouteDeclaration(None=MaterialFunctionCall(
    UVs=MaterialFunctionCall(UVs=@n4, "Rotation Angle (0-1)"=@n5.0-1 Loop) {materialFunction=&CustomRotator}.Rotated Values
    Size=MaterialFunctionCall(Speed=@n6) {materialFunction=&MF_SparksUI_LinearTime}.0-1 Loop
    OutlineThickness=NamedRerouteUsage()
   ) {materialFunction=&MF_UI_Box}.Outline) {name=Square 2, nodeColor=<rgba 0 1 0.6719 1>}
  n7 = TextureCoordinate()
  n8 = MaterialFunctionCall(Speed=Constant() {r=0.1}) {materialFunction=&MF_SparksUI_LinearTime}
  n9 = Constant() {r=0.3}
  NamedRerouteDeclaration(None=MaterialFunctionCall(
    UVs=MaterialFunctionCall(UVs=@n7, "Rotation Angle (0-1)"=@n8.0-1 Loop) {materialFunction=&CustomRotator}.Rotated Values
    Size=MaterialFunctionCall(Speed=@n9) {materialFunction=&MF_SparksUI_LinearTime}.0-1 Loop
    OutlineThickness=NamedRerouteUsage()
   ) {materialFunction=&MF_UI_Box}.Outline) {name=Square 3, nodeColor=<rgba 1 0.3516 0 1>}
  NamedRerouteDeclaration(None=Constant() {r=0.01}) {name=Outline Thickness, nodeColor=<rgba 0.0312 0 1 1>}
  NamedRerouteDeclaration(None=VectorParameter() {parameterName=Color, defaultValue=<rgba 0.0337 0.4864 1 1>}.RGB) {name=Color, nodeColor=<rgba 0.7344 0 1 1>}
  n10 = MaterialFunctionCall(
   UVs=TextureCoordinate()
   "Rotation Angle (0-1)"=MaterialFunctionCall(Speed=Constant() {r=0.11}) {materialFunction=&MF_SparksUI_LinearTime}.0-1 Loop
  ) {materialFunction=&CustomRotator}
  n11 = MaterialFunctionCall(Speed=Constant() {r=0.33}) {materialFunction=&MF_SparksUI_LinearTime}
  n12 = NamedRerouteUsage()
  n13 = NamedRerouteUsage()
  n14 = Constant() {r=0.12}
  n15 = MaterialFunctionCall(
   UVs=MaterialFunctionCall(
    UVs=TextureCoordinate()
    "Rotation Angle (0-1)"=MaterialFunctionCall(Speed=@n14) {materialFunction=&MF_SparksUI_LinearTime}.0-1 Loop
   ) {materialFunction=&CustomRotator}.Rotated Values
   Size=MaterialFunctionCall(Speed=Constant() {r=0.36}) {materialFunction=&MF_SparksUI_LinearTime}.0-1 Loop
   OutlineThickness=NamedRerouteUsage()
   GlowMax=NamedRerouteUsage()
  ) {materialFunction=&MF_UI_Box}
  n16 = Constant() {r=0.1}
  n17 = MaterialFunctionCall(
   UVs=MaterialFunctionCall(
    UVs=TextureCoordinate()
    "Rotation Angle (0-1)"=MaterialFunctionCall(Speed=@n16) {materialFunction=&MF_SparksUI_LinearTime}.0-1 Loop
   ) {materialFunction=&CustomRotator}.Rotated Values
   Size=MaterialFunctionCall(Speed=Constant() {r=0.3}) {materialFunction=&MF_SparksUI_LinearTime}.0-1 Loop
   OutlineThickness=NamedRerouteUsage()
   GlowMax=NamedRerouteUsage()
  ) {materialFunction=&MF_UI_Box}
  NamedRerouteDeclaration(None=Max(
    A=MaterialFunctionCall(UVs=@n10.Rotated Values, Size=@n11.0-1 Loop, OutlineThickness=@n12, GlowMax=@n13) {materialFunction=&MF_UI_Box}.Outline
    B=Max(A=@n15.Outline, B=@n17.Outline)
   )) {name=Glow_Opacity, nodeColor=<rgba 0 1 0.4141 1>}
  n18 = NamedRerouteUsage()
  n19 = NamedRerouteUsage()
  NamedRerouteDeclaration(None=Max(A=NamedRerouteUsage(), B=Max(A=@n18, B=@n19))) {name=Center_Line, nodeColor=<rgba 1 0.0938 0 1>}
  NamedRerouteDeclaration(None=Constant() {r=0.05}) {name=GlowMax, nodeColor=<rgba 1 0.0938 0 1>}
  n20 = Constant() {r=0.5}
  n21 = NamedRerouteUsage()
  Reroute(None=Multiply(
    A=MaterialFunctionCall(GlowMax=@n20) {materialFunction=&MF_UI_Circle}.Fill
    B=Multiply(A=@n21) {constB=0.2}
   ))
 ]
}
