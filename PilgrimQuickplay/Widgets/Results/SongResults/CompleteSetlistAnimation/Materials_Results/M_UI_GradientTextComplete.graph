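# UI-domain material for a two-color vertical text gradient in the Pilgrim Quickplay setlist-complete animation on the song results screen.
# The engine's LinearGradient function supplies a vertical gradient, which a SmoothStep tightens into a softer-edged middle band.
# That value blends between the Color Top and Color Down vector parameters, both warm gold by default.
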Material /PilgrimQuickplay/Widgets/Results/SongResults/CompleteSetlistAnimation/Materials_Results/M_UI_GradientTextComplete
graph {
 properties {bAutomaticallySetUsageInEditor=false, materialDomain=MD_UI, usedShadingModels=Default Lit}
 nodes [LinearInterpolate(
   A=VectorParameter() {parameterName=Color Top, defaultValue=<rgba 1 0.7835 0.2122 1>}.RGB
   B=VectorParameter() {parameterName=Color Down, defaultValue=<rgba 1 0.6038 0 1>}.RGB
   Alpha=SmoothStep(Value=MaterialFunctionCall() {materialFunction=&LinearGradient}.VGradient) {constMax=0.8, constMin=0.2}
  )]
}
