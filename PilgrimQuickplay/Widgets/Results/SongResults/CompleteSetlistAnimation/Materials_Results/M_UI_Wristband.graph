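# Translucent UI material that draws a wristband for the complete-setlist animation on the Quickplay song results screen.
# MF_UI_Box shapes the band fill, outline and two end detail blocks, with symbols from a channel-packed details texture cut out.
# Animated stripes, distorted by panning noise, provide a faint overlay on a screen-blended base.
# A two-color radial gradient and a label color are exposed as parameters, alongside line and outline controls.
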Material /PilgrimQuickplay/Widgets/Results/SongResults/CompleteSetlistAnimation/Materials_Results/M_UI_Wristband
graph {
 properties {bAutomaticallySetUsageInEditor=false, blendMode=BLEND_Translucent, materialDomain=MD_UI, usedShadingModels=Default Lit}
 nodes [
  n1 = Multiply(A=ComponentMask(None=NamedRerouteUsage()) {r=true}, B=ScalarParameter() {defaultValue=0.05, parameterName=LinesOpacity}) {constB=0.05}
  n2 = MaterialFunctionCall!(Blend=NamedRerouteUsage(), Base=Multiply(A=NamedRerouteUsage()) {constB=0.1}) {materialFunction=/Engine/Functions/Engine_MaterialFunctions03/Blends/Blend_Screen}
  n3 = VectorParameter() {group=Color, parameterName=Color Label, defaultValue=<rgba 1 0.4397 0.6867 1>}
  LinearInterpolate(A=LinearInterpolate(A=Add(A=@n1, B=@n2.Result), B=@n3.RGB, Alpha=NamedRerouteUsage()), B=@n3.RGB, Alpha=NamedRerouteUsage())
  n4 = Add(A=Add(A=NamedRerouteUsage(), B=NamedRerouteUsage()), B=Add(A=NamedRerouteUsage(), B=NamedRerouteUsage()))
  n5 = Add(A=NamedRerouteUsage(), B=AppendVector(A=Constant() {r=0.445}, B=Constant()))
  n6 = Constant2Vector() {g=1, r=0.095}
  NamedRerouteDeclaration(None=Min(A=OneMinus(None=@n4), B=MaterialFunctionCall(UVs=@n5, Size=@n6) {materialFunction=&MF_UI_Box}.Fill)) {name=DetailLeft, nodeColor=<rgba 0 1 0.625 1>}
  n7 = Add(A=Add(A=NamedRerouteUsage(), B=NamedRerouteUsage()), B=NamedRerouteUsage())
  n8 = Add(A=NamedRerouteUsage(), B=AppendVector(A=Constant() {r=-0.5}, B=Constant()))
  n9 = Constant2Vector() {g=0.15, r=0.1}
  NamedRerouteDeclaration(None=Min(A=OneMinus(None=@n7), B=MaterialFunctionCall(UVs=@n8, Size=@n9) {materialFunction=&MF_UI_Box}.Fill)) {name=DetailRight, nodeColor=<rgba 0 0.9062 1 1>}
  n10 = NamedRerouteUsage()
  n11 = Constant2Vector() {r=0.3}
  n12 = TextureSample(UVs=Add(A=NamedRerouteUsage(), B=Subtract(A=@n10, B=@n11) {constB=0.1})) {
   samplerSource=SSM_Clamp_WorldGroupSettings
   texture=/PilgrimQuickplay/Widgets/Results/SongResults/CompleteSetlistAnimation/Textures/T_UI_Wristband_Details
  }
  NamedRerouteDeclaration(None=@n12.R) {name=LeftSymbol_02, nodeColor=<rgba 1 0 0.8438 1>}
  n13 = NamedRerouteUsage()
  n14 = Constant2Vector() {g=-0.4}
  n15 = TextureSample(UVs=Add(A=NamedRerouteUsage(), B=Subtract(A=@n13, B=@n14))) {
   samplerSource=SSM_Clamp_WorldGroupSettings
   texture=&T_UI_Wristband_Details
  }
  NamedRerouteDeclaration(None=@n15.G) {name=RightSymbol_03, nodeColor=<rgba 0.4766 0 1 1>}
  n16 = NamedRerouteUsage()
  n17 = Constant2Vector() {r=0.6}
  n18 = TextureSample(UVs=Add(A=NamedRerouteUsage(), B=Subtract(A=@n16, B=@n17) {constB=0.1})) {
   samplerSource=SSM_Clamp_WorldGroupSettings
   texture=&T_UI_Wristband_Details
  }
  NamedRerouteDeclaration(None=@n18.R) {name=LeftSymbol_03, nodeColor=<rgba 1 0 0.8438 1>}
  n19 = NamedRerouteUsage()
  n20 = Constant2Vector() {r=0.9}
  n21 = TextureSample(UVs=Add(A=NamedRerouteUsage(), B=Subtract(A=@n19, B=@n20) {constB=0.1})) {
   samplerSource=SSM_Clamp_WorldGroupSettings
   texture=&T_UI_Wristband_Details
  }
  NamedRerouteDeclaration(None=@n21.R) {name=LeftSymbol_04, nodeColor=<rgba 1 0 0.8438 1>}
  NamedRerouteDeclaration(None=LinearInterpolate(
    A=VectorParameter() {group=Color, parameterName=Color 01, defaultValue=<rgba 1 0 0.7346 1>}.RGB
    B=VectorParameter() {group=Color, parameterName=Color 02, defaultValue=<rgba 0.7188 0 0.3812 1>}.RGB
    Alpha=MaterialFunctionCall() {materialFunction=&MF_UI_RadialGradient}.Smooth
   )) {name=Color, nodeColor=<rgba 1 0 0.9609 1>}
  NamedRerouteDeclaration(None=Constant2Vector() {r=7.85}) {name=LeftSymbol_Position, nodeColor=<rgba 0.6875 1 0 1>}
  NamedRerouteDeclaration(None=MaterialFunctionCall(UVs=NamedRerouteUsage(), Texture Scale=Constant2Vector() {g=0.04, r=0.04}) {materialFunction=&ScaleUVsByCenter}.UVs) {name=RightSymbolTextureScale, nodeColor=<rgba 0.9922 0 1 1>}
  NamedRerouteDeclaration(None=Constant2Vector() {r=-11.9}) {name=RightSymbol_Position, nodeColor=<rgba 0.2187 1 0 1>}
  NamedRerouteDeclaration(None=TextureCoordinate()) {name=TexCoord, nodeColor=<rgba 1 0.1172 0 1>}
  NamedRerouteDeclaration(None=MaterialFunctionCall(UVs=NamedRerouteUsage(), Texture Scale=Constant2Vector() {g=0.06, r=0.06}) {materialFunction=&ScaleUVsByCenter}.UVs) {name=LeftSymbolTextureScale, nodeColor=<rgba 0 0.2031 1 1>}
  n22 = TextureSample(UVs=Add(A=NamedRerouteUsage(), B=NamedRerouteUsage())) {
   samplerSource=SSM_Clamp_WorldGroupSettings
   texture=&T_UI_Wristband_Details
  }
  NamedRerouteDeclaration(None=@n22.G) {name=RightSymbol_01, nodeColor=<rgba 0.4766 0 1 1>}
  n23 = TextureCoordinate()
  n24 = AppendVector(A=Constant() {r=1}, B=Constant() {r=1.4})
  n25 = TextureSampleParameter2D(UVs=Panner(Coordinate=Multiply(A=@n23, B=@n24)) {speedX=0.015, speedY=0.04}) {
   group=Distortion
   parameterName="Distortion  Tex"
   samplerType=SAMPLERTYPE_Grayscale
   sortPriority=0
   texture=/PilgrimQuickplay/Widgets/Results/SongResults/CompleteSetlistAnimation/Textures/T_UI_LowResBlurredNoise
  }
  n26 = NamedRerouteUsage()
  n27 = Add(
   A=Multiply(
    A=ComponentMask(None=@n25.R) {g=true, r=true}
    B=ScalarParameter() {defaultValue=4, group=Distortion, parameterName=Distortion Intensity, sortPriority=2}
   )
   B=@n26
  )
  n28 = Multiply(A=ComponentMask(None=@n26) {g=true}, B=ScalarParameter() {defaultValue=25, parameterName=Number of Lines})
  n29 = Subtract(A=Frac(None=Add(A=@n27, B=@n28))) {constB=0.5}
  NamedRerouteDeclaration(None=SmoothStep(
    Min=ScalarParameter() {defaultValue=0.2, parameterName=Line Thickness}
    Max=ScalarParameter() {defaultValue=0.21, parameterName=Line Blur}
    Value=Abs(None=@n29)
   )) {name=LinesAnimation, nodeColor=<rgba 0.2656 0 1 1>}
  n30 = NamedRerouteUsage()
  n31 = Constant2Vector() {g=0.4}
  n32 = TextureSample(UVs=Add(A=NamedRerouteUsage(), B=Subtract(A=@n30, B=@n31))) {
   samplerSource=SSM_Clamp_WorldGroupSettings
   texture=&T_UI_Wristband_Details
  }
  NamedRerouteDeclaration(None=@n32.G) {name=RightSymbol_02, nodeColor=<rgba 0.4766 0 1 1>}
  n33 = MaterialFunctionCall(
   UVs=NamedRerouteUsage()
   Size=Constant2Vector() {g=0.09, r=1}
   OutlineThickness=ScalarParameter() {defaultValue=0.01, parameterName=OutlineThickness}
  ) {materialFunction=&MF_UI_Box}
  NamedRerouteDeclaration(None=@n33.Fill) {name=ShapeMask, nodeColor=<rgba 0 1 0.9063 1>}
  NamedRerouteDeclaration(None=@n33.Outline) {name=BorderStroke, nodeColor=<rgba 1 0 0.1406 1>}
  n34 = NamedRerouteUsage()
  n35 = Constant2Vector()
  n36 = TextureSample(UVs=Add(A=NamedRerouteUsage(), B=Subtract(A=@n34, B=@n35) {constB=0.1})) {
   samplerSource=SSM_Clamp_WorldGroupSettings
   texture=&T_UI_Wristband_Details
  }
  NamedRerouteDeclaration(None=@n36.R) {name=LeftSymbol_01, nodeColor=<rgba 1 0 0.8437 1>}
  NamedRerouteUsage()
 ]
}
