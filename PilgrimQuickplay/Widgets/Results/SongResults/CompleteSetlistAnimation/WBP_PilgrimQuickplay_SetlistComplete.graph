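# Full-screen "Setlist Complete" celebration overlay for Pilgrim Quickplay curated setlists, with a blurred background, animated wristband banner, glows and outlined text.
# Construct adapts text to the viewport aspect ratio via header style tags and outline font scaling, then starts IntroBG, IntroBand and a completion sound.
# When IntroBand finishes, IntroBG plays in reverse as the outro, then DynamicUIManager removes the overlay scene from the first local player.
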WidgetBlueprint /PilgrimQuickplay/Widgets/Results/SongResults/CompleteSetlistAnimation/WBP_PilgrimQuickplay_SetlistComplete
animations [
 IntroBG {
  binding_count=4
  length_seconds=0.45
  object=&self:IntroBG
  sequence {bindings [
    BG_GradientSmoke {
     id=#1
     display_name=BG_Effect
     possessed_class=Image
     tracks [{
       type=MovieSceneFloatTrack
       display_name=Render Opacity
       sections [{type=MovieSceneFloatSection, end_frame=8, start_frame=0, channels [None {type=MovieSceneScriptingFloatChannel, keys [[0,0],[8,1]]}]}]
      }]
    }
    BG_CenterGlow {
     id=#2
     display_name=BG_Glow
     possessed_class=Image
     tracks [{
       type=MovieSceneFloatTrack
       display_name=Render Opacity
       sections [{type=MovieSceneFloatSection, end_frame=8, start_frame=4, channels [None {type=MovieSceneScriptingFloatChannel, keys [[4,0],[8,1]]}]}]
      }]
    }
    BG_CenterHardGlow {
     id=#3
     possessed_class=Image
     tracks [{
       type=MovieScene2DTransformTrack
       sections [{
         type=MovieScene2DTransformSection
         end_frame=9
         start_frame=0
         channels [
          Scale.X {type=MovieSceneScriptingFloatChannel, const [3,1]}
          Scale.Y {type=MovieSceneScriptingFloatChannel, keys [[0,0],[9,4]]}
         ]
        }]
      }]
    }
    BackgroundBlur {
     id=#4
     display_name=BackgroundBlur_96
     possessed_class=BackgroundBlur
     tracks [{
       type=MovieSceneFloatTrack
       display_name=Render Opacity
       sections [{type=MovieSceneFloatSection, end_frame=4, start_frame=0, channels [None {type=MovieSceneScriptingFloatChannel, keys [[0,0],[4,1]]}]}]
      }]
    }
   ]}
 }
 IntroBand {
  binding_count=13
  length_seconds=4.2
  object=&self:IntroBand
  sequence {bindings [
    Glow_Right {
     id=#5
     possessed_class=Image
     tracks [
      {
       type=MovieSceneFloatTrack
       display_name=Render Opacity
       sections [{
         type=MovieSceneFloatSection
         end_frame=84
         start_frame=0
         channels [None {type=MovieSceneScriptingFloatChannel, keys [[0,0],[8,0],[14,1],[21,1],[65,0]]}]
        }]
      }
      {
       type=MovieScene2DTransformTrack
       sections [{
         type=MovieScene2DTransformSection
         end_frame=80
         start_frame=8
         channels [
          Scale.X {type=MovieSceneScriptingFloatChannel, keys [[11,0],[15,1],[65,1]]}
          Scale.Y {type=MovieSceneScriptingFloatChannel, default=0, const [1,1]}
         ]
        }]
      }
     ]
    }
    Glow_Wristband {
     id=#6
     possessed_class=Image
     tracks [
      {
       type=MovieScene2DTransformTrack
       sections [{
         type=MovieScene2DTransformSection
         end_frame=25
         start_frame=0
         channels [
          Scale.X {type=MovieSceneScriptingFloatChannel, keys [[0,0],[12,0],[14,5],[21,0]]}
          Scale.Y {type=MovieSceneScriptingFloatChannel, keys [[0,0],[12,0],[14,3],[21,0]]}
         ]
        }]
      }
      {
       type=MovieSceneFloatTrack
       display_name=Render Opacity
       sections [{
         type=MovieSceneFloatSection
         end_frame=25
         start_frame=0
         channels [None {type=MovieSceneScriptingFloatChannel, default=0.5, keys [[10,0],[21,1]]}]
        }]
      }
     ]
    }
    Outline_Text {
     id=#7
     possessed_class=VerticalBox
     tracks [
      {
       type=MovieSceneFloatTrack
       display_name=Render Opacity
       sections [{
         type=MovieSceneFloatSection
         end_frame=80
         start_frame=0
         channels [None {type=MovieSceneScriptingFloatChannel, default=0, keys [[0,0],[7.68,0],[16.2137,1],[60,0]]}]
        }]
      }
      {
       type=MovieScene2DTransformTrack
       sections [{
         type=MovieScene2DTransformSection
         end_frame=80
         start_frame=0
         channels [
          Scale.X {type=MovieSceneScriptingFloatChannel, default=1, keys [[0,0],[10.24,0],[11.2167,0.6],[59.7343,1]]}
          Scale.Y {type=MovieSceneScriptingFloatChannel, default=1, keys [[0,0],[10.24,0],[10.6227,0.3],[11.2167,0.5],[59.7343,1]]}
         ]
        }]
      }
     ]
    }
    Glow_Left {
     id=#8
     possessed_class=Image
     tracks [
      {
       type=MovieScene2DTransformTrack
       sections [{
         type=MovieScene2DTransformSection
         end_frame=80
         start_frame=8
         channels [
          Scale.X {type=MovieSceneScriptingFloatChannel, keys [[11,0],[15,1],[65,1],[80,0]]}
          Scale.Y {type=MovieSceneScriptingFloatChannel, default=0, const [1,1]}
         ]
        }]
      }
      {
       type=MovieSceneFloatTrack
       display_name=Render Opacity
       sections [{
         type=MovieSceneFloatSection
         end_frame=84
         start_frame=0
         channels [None {type=MovieSceneScriptingFloatChannel, keys [[0,0],[8,0],[14,1],[21,1],[65,0]]}]
        }]
      }
     ]
    }
    Triangles_Right {
     id=#9
     possessed_class=Image
     tracks [
      {
       type=MovieSceneFloatTrack
       display_name=Render Opacity
       sections [{
         type=MovieSceneFloatSection
         end_frame=80
         start_frame=0
         channels [None {type=MovieSceneScriptingFloatChannel, keys [[0,0],[14.9153,0],[21.695,1],[60,0]]}]
        }]
      }
      {type=MovieSceneWidgetMaterialTrack, sections []}
     ]
    }
    BG_DiamondAnim {
     id=#10
     possessed_class=Image
     tracks [
      {
       type=MovieSceneFloatTrack
       display_name=Render Opacity
       sections [{type=MovieSceneFloatSection, end_frame=81, start_frame=1, channels [None {type=MovieSceneScriptingFloatChannel, keys [[1,0],[4,1],[68,0]]}]}]
      }
      {
       type=MovieScene2DTransformTrack
       sections [{
         type=MovieScene2DTransformSection
         end_frame=81
         start_frame=1
         channels [
          Scale.X {type=MovieSceneScriptingFloatChannel, keys [[1,0],[6,0],[38,1.5]]}
          Scale.Y {type=MovieSceneScriptingFloatChannel, keys [[1,0],[6,0],[38,1.5]]}
         ]
        }]
      }
     ]
    }
    Text_MainHeader {
     id=#11
     possessed_class=CommonTextBlock
     tracks [
      {
       type=MovieSceneFloatTrack
       display_name=Render Opacity
       sections [{
         type=MovieSceneFloatSection
         end_frame=59
         start_frame=0
         channels [None {type=MovieSceneScriptingFloatChannel, keys [[0,0],[8,0],[12,1],[56,1],[59,0]]}]
        }]
      }
      {
       type=MovieScene2DTransformTrack
       sections [{
         type=MovieScene2DTransformSection
         end_frame=24
         start_frame=8
         channels [
          Translation.X {type=MovieSceneScriptingFloatChannel, keys [[8,-500],[12,0]]}
          Translation.Y {type=MovieSceneScriptingFloatChannel, const [0,1]}
         ]
        }]
      }
     ]
    }
    Triangles_Left {
     id=#12
     possessed_class=Image
     tracks [
      {type=MovieSceneWidgetMaterialTrack, sections []}
      {
       type=MovieSceneFloatTrack
       display_name=Render Opacity
       sections [{
         type=MovieSceneFloatSection
         end_frame=80
         start_frame=0
         channels [None {type=MovieSceneScriptingFloatChannel, keys [[0,0],[14.9153,0],[21.695,1],[60,0]]}]
        }]
      }
     ]
    }
    Wristband {
     id=#13
     possessed_class=Image
     tracks [{
       type=MovieScene2DTransformTrack
       sections [{
         type=MovieScene2DTransformSection
         end_frame=80
         start_frame=0
         channels [
          Translation.X {type=MovieSceneScriptingFloatChannel, default=-2500, keys [[0,-2500],[9,0],[59.3127,0],[64.251,3000],[80,3000]]}
          Translation.Y {type=MovieSceneScriptingFloatChannel, default=0}
         ]
        }]
      }]
    }
   ]}
 }
]
graphs [
 AdaptTextAspectRatio {nodes [
   AdaptTextAspectRatio() {is=title K2Node_FunctionEntry}
   n1 = GetViewportSize!() -> (ReturnValue_X:double, ReturnValue_Y:double)
   > n2 = Set IsLandscapeAspectRatio(IsLandscapeAspectRatio:bool="float >= float"(A=float / float(A=@n1, B=@n1.ReturnValue_Y), B=Get MinAspectRatio() -> (MinAspectRatio:double))) -> (Output_Get:bool)
   n3 = "To Float (Boolean)"(InBool=@n2)
   n4 = Get TextSizeThinMultiplier() -> (TextSizeThinMultiplier:double)
   > Set TextSizeMultiplier(TextSizeMultiplier:double="Min (Float)"(A:double=float + float(A=@n3, B=@n4), B:double=1) -> (ReturnValue:double) {is=title K2Node_CommutativeAssociativeBinaryOperator}) -> (Output_Get:double)
   > n5 = Sequence!(execute) -> (then_0, then_1, then_2:Exec)
   > UpdateWidgetStyleTags!(
    execute
    InWidget:Widget Object Reference=Get Text_MainHeader() -> (Text_MainHeader:Common Text Object Reference)
    InAddedTags:Array of Gameplay Tag Structures=Make Array()
    InRemovedTags:Array of Gameplay Tag Structures=Make Array!(
     "[0]"=Get HeadingStyleBig() -> (HeadingStyleBig:Gameplay Tag Structure)
     "[1]":Gameplay Tag Structure=Get HeadingStyleSmall() -> (HeadingStyleSmall:Gameplay Tag Structure)
    ) -> (Array)
   ) -> (then, ReturnValue:bool)
   n6 = Get IsLandscapeAspectRatio() -> (IsLandscapeAspectRatio:bool)
   n7 = Get HeadingStyleBig()
   n8 = Get HeadingStyleSmall()
   > UpdateWidgetStyleTags(
    InWidget=Get Text_MainHeader()
    InAddedTags=Make Array!("[0]"=Select(Option 0:Gameplay Tag Structure=@n8, Option 1:Gameplay Tag Structure=@n7, Index=@n6) -> (ReturnValue:Gameplay Tag Structure)) -> (Array)
    InRemovedTags=Make Array()
   )
   n9 = Set TempFontRef(
    execute=@n5.then_1
    TempFontRef:Slate Font Info Structure=Get Font!(self:Text Object Reference=Get Text_Outline_1() -> (Text_Outline_1:Common Text Object Reference)) -> (Font:Slate Font Info Structure)
   ) -> (Output_Get:Slate Font Info Structure)
   n10 = Break Slate Font Info!(SlateFontInfo:"Slate Font Info Structure (by ref)"=@n9) -> (
    FontObject:Object Reference
    FontMaterial:Object Reference
    OutlineSettings:Font Outline Settings Structure
    TypefaceFontName:Name
    Size:float=~
    LetterSpacing:int
    SkewAmount:float=~
    bForceMonospaced:bool=~
    bMaterialIsStencil:bool=~
    MonospacedWidth:float=~
   ) {is=title K2Node_BreakStruct}
   > "Set Float (double-precision) (by ref)"!(
    execute:Exec
    Target:"Float (double-precision) (by ref)"=@n10.Size
    Value:double=float * float(A=@n10.Size, B=Get TextSizeMultiplier() -> (TextSizeMultiplier:double)) {is=title K2Node_PromotableOperator}
   ) -> (then:Exec) {is=title K2Node_VariableSetRef}
   > SetFont(self=Get Text_Outline_1(), InFontInfo=Get TempFontRef() -> (TempFontRef:Slate Font Info Structure))
   n11 = Set TempFontRef(execute=@n5.then_2, TempFontRef=Get Font(self=Get Text_Outline_2() -> (Text_Outline_2:Common Text Object Reference)))
   n12 = Break Slate Font Info(SlateFontInfo=@n11) -> (Size=~, SkewAmount=~, bForceMonospaced=~, bMaterialIsStencil=~, MonospacedWidth=~)
   > "Set Float (double-precision) (by ref)"(Target=@n12.Size, Value=float * float(A=@n12.Size, B=Get TextSizeMultiplier()))
   > SetFont(self=Get Text_Outline_2(), InFontInfo=Get TempFontRef())
  ]}
 EventGraph {nodes [
   PlayBackgroundAnimIntro() {is=title K2Node_CustomEvent}
   > PlayAnimationForward(InAnimation=Get IntroBG() -> (IntroBG:Widget Animation Object Reference))
   PlayBandAnim() {is=title K2Node_CustomEvent}
   > n1 = Play Animation with Finished event!(
    execute:Exec
    Widget:User Widget Object Reference=Self-Reference()
    InAnimation:Widget Animation Object Reference=Get IntroBand() -> (IntroBand:Widget Animation Object Reference)
    StartAtTime:float=0
    NumLoopsToPlay:int=1
    PlayMode:EUMGSequencePlayMode Enum
    PlaybackSpeed:float=1
   ) -> (then:Exec, Result:Widget Animation Handle Structure, Finished:Exec) {is=title K2Node_PlayAnimation2}
   PlayBackgroundAnimOutro(execute=@n1.Finished)
   PlayBackgroundAnimOutro() {is=title K2Node_CustomEvent}
   n2 = Get IntroBG()
   > n3 = Play Animation with Finished event(
    Widget=Self-Reference()
    InAnimation=@n2
    StartAtTime=GetAnimationCurrentTime!(self:User Widget Object Reference, InAnimation:Widget Animation Object Reference=@n2) -> (ReturnValue:float)
    NumLoopsToPlay=1
    PlayMode=Reverse
    PlaybackSpeed=1
   )
   RemoveSceneFromFirstLocalPlayer!(
    execute=@n3.Finished
    self:Dynamic UIManager Object Reference=Get DynamicUIManager() -> (ReturnValue:Dynamic UIManager Object Reference) {is=title K2Node_GetSubsystem}
    Scene:Dynamic UIScene Object Reference=/PilgrimQuickplay/Widgets/Scenes/UIScene_PilgrimQuickplay_CuratedSetlist_SetlistCompleteOverlay
   ) -> (then)
   Event Construct()
   > AdaptTextAspectRatio() {is=title K2Node_CallFunction}
   > PlayBackgroundAnimIntro() {is=title K2Node_CallFunction}
   > PlayBandAnim() {is=title K2Node_CallFunction}
   > PlaySound2D(Sound=/PilgrimQuickplay/Audio/Results/MSS_FM_Results_SetListComplete)
  ]}
]
variables [
 HeadingStyleBig {type=GameplayTag, default {tagName=StyleSheet.Style.40}}
 HeadingStyleSmall {type=GameplayTag, default {tagName=StyleSheet.Style.30}}
 MinAspectRatio {type=number, default=0.6}
 TextSizeThinMultiplier {type=number, default=0.9}
]
widgets {
 info {inheritedWidgetCount=0, parentClass=&UserWidget, rootWidgetClass=&Overlay, widgetCount=24}
 widgets [
  Overlay_38 {widgetClassPath=&Overlay, widget=Overlay_38, properties {clipping=ClipToBounds}}
  BGPLACEHOLDERCOLLAPSED {
   widgetClassPath=&Image
   bIsVariable=true
   parent=Overlay_38
   widget=BGPLACEHOLDERCOLLAPSED
   properties {
    visibility=Collapsed
    brush {
     bIsDynamicallyLoaded=false
     drawAs=Image
     imageType=FullColor
     mirroring=NoMirror
     resourceName=None
     resourceObject=/PilgrimQuickplay/Widgets/Results/SongResults/CompleteSetlistAnimation/Textures/T_UI_PlacehoderBG
     tiling=NoTile
     imageSize=<1920 1025>
     margin=<ltrb 0 0 0 0>
     outlineSettings {
      bUseBrushTransparency=false
      roundingType=HalfHeightRadius
      width=0
      color {colorUseRule=UseColor_Specified, specifiedColor=<rgba 0 0 0 0>}
      cornerRadii=<0 0 0 0>
     }
     tintColor {colorUseRule=UseColor_Specified, specifiedColor=<rgba 1 1 1 1>}
     uVRegion {bIsValid=false, max=<0 0>, min=<0 0>}
    }
   }
   slot {class=OverlaySlot, properties {horizontalAlignment=HAlign_Fill, verticalAlignment=VAlign_Fill}}
  }
  BackgroundBlur {
   widgetClassPath=&BackgroundBlur
   parent=Overlay_38
   widget=BackgroundBlur
   properties {blurStrength=2}
   slot {class=OverlaySlot, properties {horizontalAlignment=HAlign_Fill, verticalAlignment=VAlign_Fill}}
  }
  WBP_CaptureForPostBufferUpdate {
   widgetClassPath=&WBP_CaptureForPostBufferUpdate
   bIsVariable=true
   parent=Overlay_38
   slot {class=OverlaySlot, properties {horizontalAlignment=HAlign_Fill, verticalAlignment=VAlign_Fill}}
  }
  BG_GradientSmoke {
   widgetClassPath=&Image
   bIsVariable=true
   parent=Overlay_38
   widget=BG_GradientSmoke
   properties {
    visibility=HitTestInvisible
    brush {
     bIsDynamicallyLoaded=false
     drawAs=Image
     imageType=FullColor
     mirroring=NoMirror
     resourceName=None
     resourceObject=/PilgrimQuickplay/Widgets/Results/SongResults/CompleteSetlistAnimation/Materials_Results/MI_UI_Background
     tiling=NoTile
     imageSize=<1920 1025>
     margin=<ltrb 0 0 0 0>
     outlineSettings {
      bUseBrushTransparency=false
      roundingType=HalfHeightRadius
      width=0
      color {colorUseRule=UseColor_Specified, specifiedColor=<rgba 0 0 0 0>}
      cornerRadii=<0 0 0 0>
     }
     tintColor {colorUseRule=UseColor_Specified, specifiedColor=<rgba 1 1 1 1>}
     uVRegion {bIsValid=false, max=<0 0>, min=<0 0>}
    }
   }
   slot {class=OverlaySlot, properties {horizontalAlignment=HAlign_Fill, verticalAlignment=VAlign_Fill}}
  }
  BG_DiamondAnim {
   widgetClassPath=&Image
   bIsVariable=true
   parent=Overlay_38
   widget=BG_DiamondAnim
   properties {brush {
     bIsDynamicallyLoaded=false
     drawAs=Image
     imageType=FullColor
     mirroring=NoMirror
     resourceName=None
     resourceObject=/PilgrimQuickplay/Widgets/Results/SongResults/CompleteSetlistAnimation/Materials_Results/M_UI_DiamondAnim
     tiling=NoTile
     imageSize=<800 800>
     margin=<ltrb 0 0 0 0>
     outlineSettings {
      bUseBrushTransparency=false
      roundingType=HalfHeightRadius
      width=0
      color {colorUseRule=UseColor_Specified, specifiedColor=<rgba 0 0 0 0>}
      cornerRadii=<0 0 0 0>
     }
     tintColor {colorUseRule=UseColor_Specified, specifiedColor=<rgba 1 1 1 1>}
     uVRegion {bIsValid=false, max=<0 0>, min=<0 0>}
    }}
   slot {class=OverlaySlot, properties {horizontalAlignment=HAlign_Center, verticalAlignment=VAlign_Center}}
  }
  BG_CenterGlow {
   widgetClassPath=&Image
   bIsVariable=true
   parent=Overlay_38
   widget=BG_CenterGlow
   properties {
    visibility=HitTestInvisible
    brush {
     bIsDynamicallyLoaded=false
     drawAs=Image
     imageType=FullColor
     mirroring=NoMirror
     resourceName=None
     resourceObject=/PilgrimQuickplay/Widgets/Results/SongResults/CompleteSetlistAnimation/Materials_Results/MI_UI_Glow
     tiling=NoTile
     imageSize=<32 32>
     margin=<ltrb 0 0 0 0>
     outlineSettings {
      bUseBrushTransparency=false
      roundingType=HalfHeightRadius
      width=0
      color {colorUseRule=UseColor_Specified, specifiedColor=<rgba 0 0 0 0>}
      cornerRadii=<0 0 0 0>
     }
     tintColor {colorUseRule=UseColor_Specified, specifiedColor=<rgba 1 1 1 1>}
     uVRegion {bIsValid=false, max=<0 0>, min=<0 0>}
    }
   }
   slot {class=OverlaySlot, properties {horizontalAlignment=HAlign_Fill, verticalAlignment=VAlign_Fill}}
  }
  BG_CenterHardGlow {
   widgetClassPath=&Image
   parent=Overlay_38
   widget=BG_CenterHardGlow
   properties {
    visibility=SelfHitTestInvisible
    brush {
     bIsDynamicallyLoaded=false
     drawAs=Image
     imageType=FullColor
     mirroring=NoMirror
     resourceName=None
     resourceObject=/PilgrimQuickplay/Widgets/Results/SongResults/CompleteSetlistAnimation/Materials_Results/MI_UI_Glow_Wristband
     tiling=NoTile
     imageSize=<256 256>
     margin=<ltrb 0 0 0 0>
     outlineSettings {
      bUseBrushTransparency=false
      roundingType=HalfHeightRadius
      width=0
      color {colorUseRule=UseColor_Specified, specifiedColor=<rgba 0 0 0 0>}
      cornerRadii=<0 0 0 0>
     }
     tintColor {colorUseRule=UseColor_Specified, specifiedColor=<rgba 1 1 1 1>}
     uVRegion {bIsValid=false, max=<0 0>, min=<0 0>}
    }
    renderTransform {angle=0, scale=<3 4>, shear=<-12 0>, translation=<0 0>}
   }
   slot {class=OverlaySlot, properties {horizontalAlignment=HAlign_Center, verticalAlignment=VAlign_Fill}}
  }
  OutlineText+Glow {
   widgetClassPath=&Overlay
   parent=Overlay_38
   widget=OutlineText+Glow
   slot {class=OverlaySlot, properties {horizontalAlignment=HAlign_Center, verticalAlignment=VAlign_Center}}
  }
  Outline_Text {
   widgetClassPath=&VerticalBox
   parent=OutlineText+Glow
   widget=Outline_Text
   slot {class=OverlaySlot, properties {horizontalAlignment=HAlign_Center, verticalAlignment=VAlign_Center}}
  }
  Text_Outline_1 {
   widgetClassPath=&CommonTextBlock
   bIsVariable=true
   parent=Outline_Text
   widget=Text_Outline_1
   properties {
    applyLineHeightToBottomLine=false
    bIsScrollingEnabled=false
    justification=Center
    text=SETLIST
    textTransformPolicy=ToUpper
    font {
     bForceMonospaced=false
     bMaterialIsStencil=false
     fontMaterial=/PilgrimQuickplay/Widgets/Results/SongResults/CompleteSetlistAnimation/Materials_Results/MI_UI_OutlineTextStyle
     fontObject=&HeadingNow
     letterSpacing=0
     monospacedWidth=1
     size=129
     skewAmount=0
     typefaceFontName=86 Bold Italic
     outlineSettings {
      bApplyOutlineToDropShadows=false
      bMiteredCorners=false
      bSeparateFillAlpha=true
      outlineMaterial=null
      outlineSize=0
      outlineColor=<rgba 1 1 1 1>
     }
    }
    shadowOffset=<0 0>
   }
   slot {class=VerticalBoxSlot}
  }
  Text_Outline_2 {
   widgetClassPath=&CommonTextBlock
   bIsVariable=true
   parent=Outline_Text
   widget=Text_Outline_2
   properties {
    applyLineHeightToBottomLine=false
    bIsScrollingEnabled=false
    justification=Center
    text=COMPLETE
    textTransformPolicy=ToUpper
    font {
     bForceMonospaced=false
     bMaterialIsStencil=false
     fontMaterial=&MI_UI_OutlineTextStyle
     fontObject=&HeadingNow
     letterSpacing=0
     monospacedWidth=1
     size=129
     skewAmount=0
     typefaceFontName=86 Bold Italic
     outlineSettings {
      bApplyOutlineToDropShadows=false
      bMiteredCorners=false
      bSeparateFillAlpha=true
      outlineMaterial=null
      outlineSize=0
      outlineColor=<rgba 1 1 1 1>
     }
    }
    shadowOffset=<0 0>
   }
   slot {class=VerticalBoxSlot}
  }
  Glow_Wristband {
   widgetClassPath=&Image
   bIsVariable=true
   parent=OutlineText+Glow
   widget=Glow_Wristband
   properties {
    renderOpacity=0.5
    visibility=SelfHitTestInvisible
    brush {
     bIsDynamicallyLoaded=false
     drawAs=Image
     imageType=FullColor
     mirroring=NoMirror
     resourceName=None
     resourceObject=&MI_UI_Glow
     tiling=NoTile
     imageSize=<900 100>
     margin=<ltrb 0 0 0 0>
     outlineSettings {
      bUseBrushTransparency=false
      roundingType=HalfHeightRadius
      width=0
      color {colorUseRule=UseColor_Specified, specifiedColor=<rgba 0 0 0 0>}
      cornerRadii=<0 0 0 0>
     }
     tintColor {colorUseRule=UseColor_Specified, specifiedColor=<rgba 1 1 1 1>}
     uVRegion {bIsValid=false, max=<0 0>, min=<0 0>}
    }
    renderTransform {angle=0, scale=<3 4>, shear=<0 0>, translation=<0 0>}
   }
   slot {class=OverlaySlot, properties {horizontalAlignment=HAlign_Center, verticalAlignment=VAlign_Center}}
  }
  Content {
   widgetClassPath=&GridPanel
   parent=Overlay_38
   widget=Content
   properties {bOverride_Cursor=true, columnFill [0.1,0.715,0.1]}
   slot {class=OverlaySlot, properties {horizontalAlignment=HAlign_Fill, verticalAlignment=VAlign_Center}}
  }
  Burst_Right {
   widgetClassPath=&Overlay
   parent=Content
   widget=Burst_Right
   properties {renderTransform {angle=0, scale=<-1 1>, shear=<0 0>, translation=<0 0>}}
   slot {class=GridSlot, properties {columnSpan=1}}
  }
  Triangles_Right {
   widgetClassPath=&Image
   bIsVariable=true
   parent=Burst_Right
   widget=Triangles_Right
   properties {brush {
     bIsDynamicallyLoaded=false
     drawAs=Image
     imageType=FullColor
     mirroring=NoMirror
     resourceName=None
     resourceObject=/PilgrimQuickplay/Widgets/Results/SongResults/CompleteSetlistAnimation/Materials_Results/MI_FX_TriangleSpiral
     tiling=NoTile
     imageSize=<200 90>
     margin=<ltrb 0 0 0 0>
     outlineSettings {
      bUseBrushTransparency=false
      roundingType=HalfHeightRadius
      width=0
      color {colorUseRule=UseColor_Specified, specifiedColor=<rgba 0 0 0 0>}
      cornerRadii=<0 0 0 0>
     }
     tintColor {colorUseRule=UseColor_Specified, specifiedColor=<rgba 1 1 1 1>}
     uVRegion {bIsValid=false, max=<0 0>, min=<0 0>}
    }}
   slot {class=OverlaySlot, properties {verticalAlignment=VAlign_Center}}
  }
  Glow_Right {
   widgetClassPath=&Image
   bIsVariable=true
   parent=Burst_Right
   widget=Glow_Right
   properties {
    brush {
     bIsDynamicallyLoaded=false
     drawAs=Image
     imageType=FullColor
     mirroring=NoMirror
     resourceName=None
     resourceObject=/PilgrimQuickplay/Widgets/Results/SongResults/CompleteSetlistAnimation/Materials_Results/MI_UI_HorizontalGlow
     tiling=NoTile
     imageSize=<400 110>
     margin=<ltrb 0 0 0 0>
     outlineSettings {
      bUseBrushTransparency=false
      roundingType=HalfHeightRadius
      width=0
      color {colorUseRule=UseColor_Specified, specifiedColor=<rgba 0 0 0 0>}
      cornerRadii=<0 0 0 0>
     }
     tintColor {colorUseRule=UseColor_Specified, specifiedColor=<rgba 1 1 1 1>}
     uVRegion {bIsValid=false, max=<0 0>, min=<0 0>}
    }
    renderTransformPivot=<0 0.5>
   }
   slot {class=OverlaySlot, properties {horizontalAlignment=HAlign_Center, verticalAlignment=VAlign_Center}}
  }
  Wirstband_Sizebox {
   widgetClassPath=&SizeBox
   parent=Content
   widget=Wirstband_Sizebox
   properties {bOverride_MinDesiredHeight=true, heightOverride=100, minDesiredHeight=100, widthOverride=1500}
   slot {class=GridSlot, properties {column=1, columnSpan=1, verticalAlignment=VAlign_Center}}
  }
  Wristband+Text {widgetClassPath=&Overlay, parent=Wirstband_Sizebox, widget=Wristband+Text, slot {class=SizeBoxSlot}}
  Wristband {
   widgetClassPath=&Image
   bIsVariable=true
   parent=Wristband+Text
   widget=Wristband
   properties {
    brush {
     bIsDynamicallyLoaded=false
     drawAs=Image
     imageType=FullColor
     mirroring=NoMirror
     resourceName=None
     resourceObject=/PilgrimQuickplay/Widgets/Results/SongResults/CompleteSetlistAnimation/Materials_Results/M_UI_Wristband
     tiling=NoTile
     imageSize=<32 32>
     margin=<ltrb 0.2 0.2 0.2 0.2>
     outlineSettings {
      bUseBrushTransparency=false
      roundingType=HalfHeightRadius
      width=0
      color {colorUseRule=UseColor_Specified, specifiedColor=<rgba 0 0 0 0>}
      cornerRadii=<0 0 0 0>
     }
     tintColor {colorUseRule=UseColor_Specified, specifiedColor=<rgba 1 1 1 1>}
     uVRegion {bIsValid=false, max=<0 0>, min=<0 0>}
    }
    renderTransform {angle=0, scale=<1 10>, shear=<0 0>, translation=<0 0>}
   }
   slot {class=OverlaySlot, properties {horizontalAlignment=HAlign_Fill, verticalAlignment=VAlign_Fill}}
  }
  Text_MainHeader {
   widgetClassPath=&CommonTextBlock
   bIsVariable=true
   parent=Wristband+Text
   widget=Text_MainHeader
   properties {
    bIsScrollingEnabled=false
    justification=Center
    text=SETLIST COMPLETE
    textTransformPolicy=ToUpper
    wrapTextAt=1170
    font {
     bForceMonospaced=false
     bMaterialIsStencil=false
     fontMaterial=/PilgrimQuickplay/Widgets/Results/SongResults/CompleteSetlistAnimation/Materials_Results/M_UI_GradientTextComplete
     fontObject=&HeadingNow
     letterSpacing=-40
     monospacedWidth=1
     size=51.75
     skewAmount=0
     typefaceFontName=86 Bold Italic
     outlineSettings {
      bApplyOutlineToDropShadows=false
      bMiteredCorners=false
      bSeparateFillAlpha=false
      outlineMaterial=null
      outlineSize=0
      outlineColor=<rgba 0 0 0 0>
     }
    }
    strikeBrush {
     bIsDynamicallyLoaded=false
     drawAs=Image
     imageType=FullColor
     mirroring=NoMirror
     resourceName=None
     resourceObject=null
     tiling=NoTile
     imageSize=<32 32>
     margin=<ltrb 0 0 0 0>
     outlineSettings {
      bUseBrushTransparency=false
      roundingType=HalfHeightRadius
      width=0
      color {colorUseRule=UseColor_Specified, specifiedColor=<rgba 0 0 0 0>}
      cornerRadii=<0 0 0 0>
     }
     tintColor {colorUseRule=UseColor_Specified, specifiedColor=<rgba 1 1 1 1>}
     uVRegion {bIsValid=false, max=<0 0>, min=<0 0>}
    }
   }
   slot {class=OverlaySlot, properties {horizontalAlignment=HAlign_Center, verticalAlignment=VAlign_Center}}
  }
  Burst_Left {widgetClassPath=&Overlay, parent=Content, widget=Burst_Left, slot {class=GridSlot, properties {column=2, columnSpan=1}}}
  Triangles_Left {
   widgetClassPath=&Image
   bIsVariable=true
   parent=Burst_Left
   widget=Triangles_Left
   properties {brush {
     bIsDynamicallyLoaded=false
     drawAs=Image
     imageType=FullColor
     mirroring=NoMirror
     resourceName=None
     resourceObject=&MI_FX_TriangleSpiral
     tiling=NoTile
     imageSize=<200 90>
     margin=<ltrb 0 0 0 0>
     outlineSettings {
      bUseBrushTransparency=false
      roundingType=HalfHeightRadius
      width=0
      color {colorUseRule=UseColor_Specified, specifiedColor=<rgba 0 0 0 0>}
      cornerRadii=<0 0 0 0>
     }
     tintColor {colorUseRule=UseColor_Specified, specifiedColor=<rgba 1 1 1 1>}
     uVRegion {bIsValid=false, max=<0 0>, min=<0 0>}
    }}
   slot {class=OverlaySlot, properties {verticalAlignment=VAlign_Center}}
  }
  Glow_Left {
   widgetClassPath=&Image
   bIsVariable=true
   parent=Burst_Left
   widget=Glow_Left
   properties {
    brush {
     bIsDynamicallyLoaded=false
     drawAs=Image
     imageType=FullColor
     mirroring=NoMirror
     resourceName=None
     resourceObject=&MI_UI_HorizontalGlow
     tiling=NoTile
     imageSize=<400 110>
     margin=<ltrb 0 0 0 0>
     outlineSettings {
      bUseBrushTransparency=false
      roundingType=HalfHeightRadius
      width=0
      color {colorUseRule=UseColor_Specified, specifiedColor=<rgba 0 0 0 0>}
      cornerRadii=<0 0 0 0>
     }
     tintColor {colorUseRule=UseColor_Specified, specifiedColor=<rgba 1 1 1 1>}
     uVRegion {bIsValid=false, max=<0 0>, min=<0 0>}
    }
    renderTransformPivot=<0 0.5>
   }
   slot {class=OverlaySlot, properties {verticalAlignment=VAlign_Center}}
  }
 ]
}
